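; Vehicle cosmetic actor for the Sparky boost, toggling its Niagara effect, audio behavior and tick as boosting starts and stops.
; On cosmetic apply, it feeds the effect port count, single-boost socket position, frontend or playlist mode, and a per-car DataTable offset.
; When color is ready, it sets painted Niagara variables and tints the chassis boost material from the paint table, or uses the default color with optional rainbow curve-atlas panning.
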(event Vehicle|Cosmetics|EventOnBoostStarted
  (Actor|Tick|SetActorTickEnabled true)
  (Components|Activation|SetActive (Variables|Default|GetNS_Boost) true false)
  (AudioGameplayBehavior|Start (select (Utilities|IsValid (Variables|Vehicle|Cosmetics|GetBoostAGBComponent) :!class /Script/Engine.KismetSystemLibrary) (Variables|Vehicle|Cosmetics|GetBoostAGBComponent) 0)))

(event Vehicle|Cosmetics|EventOnBoostFinished
  (Actor|Tick|SetActorTickEnabled false)
  (Components|Activation|SetActive (Variables|Default|GetNS_Boost) false false)
  (AudioGameplayBehavior|Stop (select (Utilities|IsValid (Variables|Vehicle|Cosmetics|GetBoostAGBComponent) :!class /Script/Engine.KismetSystemLibrary) (Variables|Vehicle|Cosmetics|GetBoostAGBComponent) 0)))

(event EventEndPlay
  (params (EndPlayReason :type EEndPlayReason%20Enum :default "Destroyed"))
  (Actor|Tick|SetActorTickEnabled false)
  (Components|Activation|SetActive (Variables|Default|GetNS_Boost) false false)
  (AudioGameplayBehavior|Stop (select (Utilities|IsValid (Variables|Vehicle|Cosmetics|GetBoostAGBComponent) :!class /Script/Engine.KismetSystemLibrary) (Variables|Vehicle|Cosmetics|GetBoostAGBComponent) 0)))

(event Vehicle|Cosmetics|EventOnCosmeticApplied
  (bind _ns_boost (Variables|Default|GetNS_Boost))
  (Niagara|SetNiagaraSystemAsset _ns_boost (Variables|Default|GetNiagaraSystem) true)
  (CallFunction|SetupBoostCount)
  (CallFunction|SingleBoostSetup)
  (bind _bisfrontend (CallFunction|SetupFrontEndPreview _ns_boost))
  (CallFunction|DetectPlaylist _bisfrontend)
  (CallFunction|OffsetPerCar (Variables|Vehicle|Cosmetics|GetVehicleCosmeticsInterface) (Variables|Default|GetNS_Boost) "DTCarOffset"))

(event Vehicle|Cosmetics|EventOnColorReady
  (params (Color :type Name :default "None") (AdditionalInfo :type Vehicle%20Cosmetics%20Additional%20Parameter%20Infos%20Structure))
  (CallFunction|SetupPaintSettings Color AdditionalInfo))

(fn DetectPlaylist
  (params (isFrontEnd :type Boolean))
  (if isFrontEnd
    (LocalVariables|SetGameMode 0)
    (elif (Class|FortGameStateAthena|IsBattleRoyalePlaylist (Fortnite|Athena|GameState|GetGameStateAthena) true)
      (LocalVariables|SetGameMode 1)
      (else
        (LocalVariables|SetGameMode 2))))
  (Niagara|SetNiagaraVariableByString(Int32) (Variables|Default|GetNS_Boost) "GameModeID" (LocalVariables|GetGameMode)))

(fn OffsetPerCar
  (params (VehicleCosmeticsInterface :type Vehicle%20Cosmetics%20Interface%20Interface) (Boost :type Niagara%20Particle%20System%20Component%20Object%20Reference) (OffsetVariableName :type Name))
  (if (and (Utilities|IsValidInterface VehicleCosmeticsInterface) (Variables|Default|GetUseOffsetPerCar))
    (bind _returnvalue (Utilities|GetDataTableRow "/VehicleCosmetics/DataTables/DT_VehicleCosmetics_TurboOffsets.DT_VehicleCosmetics_TurboOffsets" (Vehicle|Cosmetics|GetBodyItemName VehicleCosmeticsInterface)))
    (Niagara|SetNiagaraVariable(Vector3) Boost OffsetVariableName (Utilities|Struct|BreakStructVehiclesCosmeticsTurboOffset _returnvalue))
    (return)))

(fn SetupBoostCount
  (Niagara|SetNiagaraVariable(Int32) (Variables|Default|GetNS_Boost) "PortCount" (Variables|Vehicle|Cosmetics|GetBoosterPortsNum)))

(fn SetupFrontEndPreview
  (params (NS_Boost :type Niagara%20Particle%20System%20Component%20Object%20Reference))
  (bind _returnvalue (Reactive|Frontend|IsCosmeticObjectDisplayedinFrontendContext 0))
  (Niagara|SetNiagaraVariable(Bool) NS_Boost "User.bIsFrontendPreview" _returnvalue)
  (return _returnvalue))

(fn SetupPaintSettings
  (params (Color :type Name) (AdditionalInfo :type Vehicle%20Cosmetics%20Additional%20Parameter%20Infos%20Structure))
  (bind _rainbow_color_config (Variables|Default|GetRainbow%20Color%20Config))
  (bind (_userainbowcolor__15_b6b860804ef35221ff4d18b588c508e3 _curvecolor_18_8ba06dbb45c1243af833a9a40b1cf122 _speedpanner_23_dde5f68c4714fe13b66a5aa2607abec1 _curveatlas_21_eada49934f84c04172700d99725c3801) (Utilities|Struct|BreakStructRainbowConfig _rainbow_color_config))
  (bind _returnvalue (Vehicle|Cosmetics|GetBodyMeshComponent))
  (if (Utilities|String|EqualExactly(String) (Utilities|String|ToString(Name) Color) "None")
    (Niagara|SetNiagaraVariable(Bool) (Variables|Default|GetNS_Boost) "bIsPainted" false)
    ;[1 1 1] PaintChassis
    (Rendering|Material|SetVectorParameterValueonMaterials _returnvalue "Chassis_BoostColor" (Math|Conversions|ToVector(LinearColor) (Variables|Default|GetDefaultPortColor)))
    (if _userainbowcolor__15_b6b860804ef35221ff4d18b588c508e3
      (Rendering|Material|SetScalarParameterValueonMaterials _returnvalue "RainbowColor" 1.0)
      (bind _position (Math|Curves|GetCurvePosition _curveatlas_21_eada49934f84c04172700d99725c3801 _curvecolor_18_8ba06dbb45c1243af833a9a40b1cf122))
      (Rendering|Material|SetScalarParameterValueonMaterials _returnvalue "CurveAtlas" _position)
      (Rendering|Material|SetScalarParameterValueonMaterials _returnvalue "RainbowPannerSpeed" _speedpanner_23_dde5f68c4714fe13b66a5aa2607abec1))
    (else
      (Niagara|SetNiagaraVariable(Bool) (Variables|Default|GetNS_Boost) "bIsPainted" true)
      (Niagara|SetNiagaraVariable(Float) (Variables|Default|GetNS_Boost) "PaintedScalar" (Variables|Default|GetPaintedScalar))
      (bind _color_index (Class|BPVehicleCosmeticsFunctions|BPGetNSIndexbyColor Color))
      (Niagara|SetNiagaraVariable(Int32) (Variables|Default|GetNS_Boost) "ColorVariantID" _color_index)
      (Class|BPVehicleCosmeticsFunctionLibrary|CosmeticActorColorReady (Variables|GetColor) "PaintedColor" (Variables|GetAdditionalInfo) (Variables|Default|GetNS_Boost) (Variables|Vehicle|Cosmetics|GetVehicleCosmeticsInterface) (Variables|Default|GetDefaultPortColor))
      (bind _returnvalue_2 (Utilities|GetDataTableRow "/VehicleCosmetics/DataTables/DT_VehicleCosmetics_Painted.DT_VehicleCosmetics_Painted" (Variables|GetColor)))
      (Rendering|Material|SetVectorParameterValueonMaterials _returnvalue "Chassis_BoostColor" (Math|Conversions|ToVector(LinearColor) (Utilities|Struct|BreakStructVehicleCosmeticsPainted _returnvalue_2)))
      (return (Utilities|Struct|BreakStructVehicleCosmeticsPainted _returnvalue_2) (.Balanced_34_F677FC9D4B1B4146E8CC9DBCB9E0943E (Utilities|Struct|BreakStructVehicleCosmeticsPainted _returnvalue_2)) (.Metallic_41_F57CE68E419E301DD160BEB5D0E32BFC (Utilities|Struct|BreakStructVehicleCosmeticsPainted _returnvalue_2))))))

(fn SingleBoostSetup
  (bind _returnvalue_location (Math|Transform|BreakTransform (Transformation|GetSocketTransform (Vehicle|Cosmetics|GetBodyMeshComponent) "Socket_Body" "RTS_ParentBoneSpace")))
  (bind (_returnvalue_location_x _returnvalue_location_y) (Math|Vector|BreakVector _returnvalue_location))
  (bind _returnvalue_location_2 (Math|Transform|BreakTransform (Transformation|GetSocketTransform (Vehicle|Cosmetics|GetBodyMeshComponent) "Socket_Turbo_A_R" "RTS_ParentBoneSpace")))
  (bind (_returnvalue_location_x_2 _returnvalue_location_y_2 _returnvalue_location_z) (Math|Vector|BreakVector _returnvalue_location_2))
  (bind _ns_boost (Variables|Default|GetNS_Boost))
  (if (Variables|Default|GetSingleBoost?)
    (Niagara|SetNiagaraVariable(Vector3) _ns_boost "SingleBoostPos" (Math|Vector|MakeVector _returnvalue_location_x_2 _returnvalue_location_y _returnvalue_location_z))
    (else
      (Niagara|SetNiagaraVariable(Vector3) _ns_boost "SingleBoostPos" "0, 0, 0"))))

(fn ConstructionScript)
