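# Niagara system for the Sparky vehicle boost cosmetic, a sparking exhaust effect trailing from the car's boost sockets.
# A burst glow sprite, short-lived white stars and SubUV electric sprites are partitioned across skeletal-mesh sockets, shifted by the car offset.
# Ribbons spawn per port at a quality-dependent rate, with curve-driven color, width and a lifted secondary position.
# GameModeID switches ribbon motion; bIsPainted removes the electric brightness flicker, while painted-variant override modules are disabled.
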NiagaraSystem /VehicleCosmetics/Boosts/Boost_Sparky/FX/NS_Boost_Sparky
dependencies {
 usedDataInterfaces [
  &NiagaraDataInterfaceArrayPosition
  &NiagaraDataInterfaceVectorCurve
  &NiagaraDataInterfaceCurve
  &NiagaraDataInterfaceColorCurve
  &NiagaraDataInterfaceSpriteRendererInfo
  &NiagaraDataInterfaceSkeletalMesh
 ]
 usedDynamicInputs []
 usedModules [
  &self:SystemSpawnScript.SystemScriptSource.SystemScriptGraph.NiagaraNodeAssignment_0.SetVariables_#1
  &SystemState
 ]
}
emitters {
 NE_VC_Boost_Sprite_Parent {
  emitterUpdateScript [
   &EmitterState(
    Inactive Response=::NewEnumerator1
    Life Cycle Mode=::NewEnumerator1
    Loop Behavior=::NewEnumerator1
    Loop Duration Mode=::NewEnumerator1
    UseLoopDelay=false
   )
   &SpawnBurst_Instantaneous(Spawn Count=4, Use Spawn Probability=false, UseLoopCountLimit=false)
   &VC_FilteredSocketCount()
   &SpawnRate(SpawnRate=50, Use Spawn Probability=false) {enabled=false}
  ]
  particleSpawnScript [
   &VC_PaintedVariantsOverride() {enabled=false}
   &InitializeParticle(
    Color=<rgba 0.8713 0 1 0.7>
    Lifetime=$Emitter.CurrentLoopDuration
    Position Mode=::NewEnumerator0
    Sprite Rotation Mode=::NewEnumerator2
    Sprite Size Mode=::NewEnumerator3
    Uniform Sprite Size=120.3781
    Write Material Random=true
   )
   &PartitionParticles(Number Of Partitions=$Emitter.PortCount)
   &SkeletalMeshLocation(
    Bone / Socket Index=$Particles.PartitionParticles.Partition
    Bone / Socket Sampling Mode=::NewEnumerator5
    MaskSpawnGroup=false
    Mesh Sampling Type=::NewEnumerator6
    Sampled Position Offset=&AddVectorToPosition(Position=$User.DTCarOffset, Vector=<-75 0 10>)
    Skeletal Mesh=$System.SkeletalMesh
   )
   &AddVelocity(Velocity=<-1500 0 0>) {enabled=false}
   /SetVariables(Particles.BlueSize:Vector2f=<70 70>, Particles.ParticleBlue:LinearColor=<rgba 0.0508 0.1562 0.4435 1.3>) {isSetParameters=true}
  ]
  particleUpdateScript [
   &ParticleState()
   &SolveForcesAndVelocity(Clamp Velocity=false, Limit Acceleration=false)
  ]
  rendererClasses [&NiagaraSpriteRendererProperties]
  renderers [
   {rendererClass=&NiagaraSpriteRendererProperties}
   {rendererClass=&NiagaraSpriteRendererProperties, rendererIndex=1}
  ]
 }
 Ribbons {
  emitterUpdateScript [
   &EmitterState()
   &VC_FilteredSocketCount()
   &VC_RateQualityHandler(High Spawn Rate=60, Low Spawn Rate=30, Use Custom Rate?=true)
   &VC_SpawnRibbon_v1(Number of Ribbons=$Emitter.PortCount, SpawnRate=$Emitter.Spawn Rate)
  ]
  particleSpawnScript [
   &VC_PaintedVariantsOverride() {enabled=false}
   &VC_InitializeRibbon_v1(
    Custom Offset=true
    Enable Skeletal Position=true
    Local Vector Offset=&AddVectorToPosition(Position=$User.DTCarOffset, Vector=<-60 0 5>)
    Offset Amount=1
    Ribbon Width=119.3218
   )
   &AddVelocity(Velocity=&MergeHelper_VectorSwitch(0=<-1850 0 -500>, 1=<-1850 0 -500>, 2=<-1850 0 -500>, 3=?, Index=$User.GameModeID, Num=::NewEnumerator1)) {module=AddVelocity001}
   /SetVariables(Particles.PositionBlue=DataInterface() {propertyValues {GpuSyncMode=SyncCpuToGpu, MaxElements=0, PositionData [<0 14.9347 0>]}}) {isSetParameters=true}
  ]
  particleUpdateScript [
   &ParticleState()
   &ScaleVelocity(
    Coordinate Space:ENiagaraCoordinateSpace=::Simulation
    Velocity Scale=&MergeHelper_VectorSwitch(
     0=&VectorFromCurve(VectorCurve=DataInterface() {propertyValues {
        XCurve {keys [<at 0 1>,<at 1 1>]}
        YCurve {keys [<at 0 1>,<at 1 1>]}
        ZCurve {keys [{arriveTangent=-3.9821, interpMode=RCIM_Cubic, leaveTangent=-3.9821, tangentMode=RCTM_User, time=0, value=0}, <at 0.8 1>, <at 1 0.8>]}
       }})
     1=&VectorFromCurve(VectorCurve=DataInterface() {propertyValues {
        XCurve {keys [<at 0 1>,<at 1 1>]}
        YCurve {keys [<at 0 1>,<at 1 1>]}
        ZCurve {keys [{arriveTangent=-3.9821, interpMode=RCIM_Cubic, leaveTangent=-3.9821, tangentMode=RCTM_User, time=0, value=0}, <at 0.8 1>, <at 1 0.8>]}
       }})
     2=&VectorFromCurve(VectorCurve=DataInterface() {propertyValues {
        XCurve {keys [<at 0 1>,<at 1 1>]}
        YCurve {keys [<at 0 1>,<at 1 1>]}
        ZCurve {keys [{arriveTangent=-3.9821, interpMode=RCIM_Cubic, leaveTangent=-3.9821, tangentMode=RCTM_User, time=0, value=0}, <at 0.8 1>, <at 1 0.8>]}
       }})
     3=?
     Index=$User.GameModeID
     Num=::NewEnumerator1
    )
   ) {enabled=false}
   &InheritSourceMovement(ApplyForce=false, ApplyPosition=true, ApplyPositionFalloff=false)
   &SolveForcesAndVelocity(Clamp Velocity=false, Limit Acceleration=false)
   &VC_UpdateRibbon_v1(Enable Skeletal Position=true)
   /SetVariables(Particles.PositionOffsetZ:NiagaraPosition=&AddVectorToPosition(
     Position=$Particles.Position
     Vector=&MergeHelper_VectorSwitch(
      0=/VehicleCosmetics/FX/Niagara/Modules/VC_ConvertVectorToWorld(Vector:Vector3f=&MakeVector(
        X=0
        Y=0
        Z=&FloatFromCurve(FloatCurve=DataInterface() {propertyValues {Curve {
            defaultValue=0
            keys [
             {arriveTangent=105.2416, interpMode=RCIM_Cubic, leaveTangent=105.2416, tangentMode=RCTM_User, time=0, value=0.3}
             {arriveTangent=205.0984, interpMode=RCIM_Cubic, leaveTangent=205.0984, tangentMode=RCTM_User, time=0.1, value=15}
             {arriveTangent=55.5556, interpMode=RCIM_Cubic, leaveTangent=55.5556, time=0.5, value=55}
             <at 1 65 RCIM_Cubic>
            ]
           }}})
       ))
      1=&VC_ConvertVectorToWorld(Vector:Vector3f=&MakeVector(
        X=0
        Y=0
        Z=&FloatFromCurve(FloatCurve=DataInterface() {propertyValues {Curve {
            defaultValue=0
            keys [
             {arriveTangent=105.2416, interpMode=RCIM_Cubic, leaveTangent=105.2416, tangentMode=RCTM_User, time=0, value=0.3}
             {arriveTangent=205.0984, interpMode=RCIM_Cubic, leaveTangent=205.0984, tangentMode=RCTM_User, time=0.1, value=15}
             {arriveTangent=55.5556, interpMode=RCIM_Cubic, leaveTangent=55.5556, time=0.5, value=55}
             <at 1 65 RCIM_Cubic>
            ]
           }}})
       ))
      2=&VC_ConvertVectorToWorld(Vector:Vector3f=&MakeVector(
        X=0
        Y=0
        Z=&FloatFromCurve(FloatCurve=DataInterface() {propertyValues {Curve {
            defaultValue=0
            keys [
             {arriveTangent=105.2416, interpMode=RCIM_Cubic, leaveTangent=105.2416, tangentMode=RCTM_User, time=0, value=0.3}
             {arriveTangent=205.0984, interpMode=RCIM_Cubic, leaveTangent=205.0984, tangentMode=RCTM_User, time=0.1, value=15}
             {arriveTangent=55.5556, interpMode=RCIM_Cubic, leaveTangent=55.5556, time=0.5, value=55}
             <at 1 65 RCIM_Cubic>
            ]
           }}})
       ))
      3=?
      Index=$User.GameModeID
      Num=::NewEnumerator1
     )
    )) {isSetParameters=true}
   &ScaleRibbonWidth(Ribbon Width Scale=&FloatFromCurve(FloatCurve=DataInterface() {propertyValues {Curve {
        defaultValue=1
        keys [{interpMode=RCIM_Cubic, tangentMode=RCTM_User, time=0, value=0.4}, {interpMode=RCIM_Cubic, tangentMode=RCTM_User, time=1, value=1.5}]
       }}}))
   &Color(Color=&ColorFromCurve(ColorCurve=DataInterface() {propertyValues {
       AlphaCurve {keys [<at 0 1 RCIM_Cubic>,<at 0.3534 1>,<at 1 1 RCIM_Cubic>]}
       BlueCurve {keys [<at 0.1372 2.2 RCIM_Cubic>,<at 0.3146 2>,<at 0.7323 1 RCIM_Cubic>]}
       GreenCurve {keys [<at 0.1372 2.1017 RCIM_Cubic>,<at 0.3146 0.0492>,<at 0.7323 0 RCIM_Cubic>]}
       RedCurve {keys [<at 0.1372 2.1356 RCIM_Cubic>,<at 0.3146 2>,<at 0.7323 0.2761 RCIM_Cubic>]}
      }})) {module=Color001}
   /SetVariables(
    Particles.ColorRibbonBlue:LinearColor=&ColorFromCurve(ColorCurve=DataInterface() {propertyValues {
       AlphaCurve {keys [<at 0 1 RCIM_Cubic>,<at 1 1 RCIM_Cubic>]}
       BlueCurve {keys [<at 0 1 RCIM_Cubic>,<at 0.5696 1 RCIM_Cubic>]}
       GreenCurve {keys [<at 0 1 RCIM_Cubic>,<at 0.5696 0.6799 RCIM_Cubic>]}
       RedCurve {keys [<at 0 1 RCIM_Cubic>,<at 0.5696 0.5 RCIM_Cubic>]}
      }})
    Particles.RibbonWidhtBlue:NiagaraFloat=&FloatFromCurve(FloatCurve=DataInterface() {propertyValues {Curve {keys [<at 0 0.2 RCIM_Cubic>,<at 1 1.1 RCIM_Cubic>]}}}, Scale Curve=100)
   ) {isSetParameters=true}
  ]
  rendererClasses [&NiagaraRibbonRendererProperties]
  renderers [
   {rendererClass=&NiagaraRibbonRendererProperties}
   {rendererClass=&NiagaraRibbonRendererProperties, rendererIndex=1}
   {rendererClass=&NiagaraRibbonRendererProperties, rendererIndex=2}
  ]
 }
 WhiteStar {
  emitterUpdateScript [
   &EmitterState(
    Life Cycle Mode=::NewEnumerator1
    Loop Behavior=::NewEnumerator1
    Loop Duration Mode=::NewEnumerator1
    Scalability Mode=::NewEnumerator1
    UseLoopDelay=false
   )
   &VC_FilteredSocketCount()
   &SpawnRate(SpawnRate=50, Use Spawn Probability=false)
  ]
  particleSpawnScript [
   &VC_PaintedVariantsOverride() {enabled=false}
   &InitializeParticle(
    Color=<rgba 19.913 0 20 1>
    Lifetime=&RandomRangeFloat(Minimum=0.1, Maximum=0.2)
    Sprite Rotation Mode=::NewEnumerator2
    Sprite Size Mode=::NewEnumerator3
    Uniform Sprite Size=&RandomRangeFloat(Minimum=60, Maximum=65)
    UsePositionOffset=true
    Write Material Random=true
   )
   &PartitionParticles(Number Of Partitions=$Emitter.PortCount)
   &SkeletalMeshLocation(
    Bone / Socket Index=$Particles.PartitionParticles.Partition
    Bone / Socket Sampling Mode=::NewEnumerator5
    MaskSpawnGroup=false
    Mesh Sampling Type=::NewEnumerator6
    Sampled Position Offset=&AddVectorToPosition(Position=$User.DTCarOffset, Vector=<-65 0 10>)
    Skeletal Mesh=$System.SkeletalMesh
   )
   &AddVelocity(Velocity=&RandomRangeVector(Minimum=<100 0 0>, Maximum=<200 1 1>), Velocity Speed Scale=-1) {enabled=false}
   /SetVariables(
    Particles.ParticleWhiteStar:LinearColor=<rgba 2 1.6979 1.9427 1>
    Particles.WhiteStarSize:Vector2f=&RandomRangeVector2D(Minimum=<50 50>, Maximum=<58 58>)
   ) {isSetParameters=true}
  ]
  particleUpdateScript [
   &ParticleState()
   &ScaleSpriteSize(Uniform Curve Sprite Scale=DataInterface() {propertyValues {Curve {keys [{tangentMode=RCTM_User, time=0, value=0.5}, {tangentMode=RCTM_User, time=1, value=1}]}}})
   &SolveForcesAndVelocity(Clamp Velocity=false, Limit Acceleration=false)
  ]
  rendererClasses [&NiagaraSpriteRendererProperties]
  renderers [
   {rendererClass=&NiagaraSpriteRendererProperties}
   {rendererClass=&NiagaraSpriteRendererProperties, rendererIndex=1}
  ]
 }
 electro {
  emitterUpdateScript [
   &EmitterState()
   &VC_FilteredSocketCount()
   &SpawnRate(SpawnRate=150, Use Spawn Probability=false)
  ]
  particleSpawnScript [
   &VC_PaintedVariantsOverride() {enabled=false}
   &InitializeParticle(
    Color=<rgba 0.2157 0 0.4 10>
    Lifetime Max=0.25
    Lifetime Min=0.1
    Lifetime Mode=::NewEnumerator1
    Sprite Rotation Mode=::NewEnumerator2
    Sprite Size Mode=::NewEnumerator1
    Uniform Sprite Size Max=70
    Uniform Sprite Size Min=50
    UsePositionOffset=true
    Write Material Random=true
   )
   &PartitionParticles(Number Of Partitions=$Emitter.PortCount)
   &SkeletalMeshLocation(
    Bone / Socket Index=$Particles.PartitionParticles.Partition
    Bone / Socket Sampling Mode=::NewEnumerator5
    MaskSpawnGroup=false
    Mesh Sampling Type=::NewEnumerator6
    Sampled Position Offset=&AddVectorToPosition(Position=$User.DTCarOffset, Vector=<-45 0 0>)
    Skeletal Mesh=$System.SkeletalMesh
   )
   &AddVelocity(Velocity=<-400 0 120>)
   &ShapeLocation(MaskSpawnGroup=false, Override Seed=false, Sphere Radius=0, UseOwnerScale=true)
  ]
  particleUpdateScript [
   &ParticleState()
   &SolveForcesAndVelocity(Clamp Velocity=false, Limit Acceleration=false)
   &SubUVAnimation(
    End Frame Range Override=0
    Random Start Frame=true
    "Sprite Renderer "=DataInterface() {propertyValues {SpriteRenderer=&self:electro_0.NiagaraSpriteRendererProperties_0}}
    Start Frame Offset:NiagaraInt32=&RandomRangeInt(Minimum=0, Maximum=5)
    SubUV Animation Mode:ENiagara_SubUVLookupModeV2=::NewEnumerator1
    UseEndFrame=true
    UseStartFrame=false
   )
   &DynamicMaterialParameters(
    Index 0 Param 2=5.5
    Index 0 Param 3=1
    Index 0 Param 4=&RandomRangeFloat(Minimum=0.1, Maximum=0.5, Evaluation Type=::NewEnumerator1)
    Module.Index0Float0Write=true
    Module.Index0Float1Write=true
    Module.Index0Float2Write=true
    Module.Index0Float3Write=true
   )
   &ScaleColor(
    Scale RGB=&Multiply_VectorByFloat(
     Vector=<1 1 1>
     Float=&MakeCustomFloatFromBool(Bool=$User.bIsPainted, True Float=1, False Float=&RandomRangeFloat(Minimum=10, Maximum=20))
    )
    ScaleA=true
    ScaleRGB=true
   )
  ]
  rendererClasses [&NiagaraSpriteRendererProperties]
  renderers [{rendererClass=&NiagaraSpriteRendererProperties}]
 }
}
summary {
 systemName=NS_Boost_Sparky
 emitters [
  {emitterName=Ribbons, rendererClasses [&NiagaraRibbonRendererProperties]}
  {emitterName=electro, rendererClasses [&NiagaraSpriteRendererProperties]}
  {emitterName=WhiteStar, rendererClasses [&NiagaraSpriteRendererProperties]}
  {emitterName=NE_VC_Boost_Sprite_Parent, rendererClasses [&NiagaraSpriteRendererProperties]}
 ]
 userVariables [
  User.bIsFrontendPreview {type=NiagaraBool}
  User.bIsPainted {type=NiagaraBool}
  User.ColorVariantID {type=NiagaraInt32}
  User.DTCarOffset {type=NiagaraPosition, description="", defaultValue {}}
  User.GameModeID {type=NiagaraInt32}
  User.MultiplierSkyColor {type=NiagaraFloat, defaultValue=1, description=""}
  User.PaintedColor {type=LinearColor, description="", defaultValue {struct=LinearColor, value=<rgba 1 1 1 1>}}
  User.PaintedScalar {type=NiagaraFloat, defaultValue=1, description=""}
  User.PortCount {type=NiagaraInt32, defaultValue=1}
 ]
}
